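# Animation Blueprint for the Dark Viking backpack's wings, posing and animating them for ground movement, skydiving and targeting.
# A state machine selects Base, Diving, Floating or Targeting poses, feeding chained AnimDynamics wing bones with alphas set by speed and dive mode.
# Perlin noise from pawn height pushes outer finger bones during skydiving, halved under parachute.
# The cached result gets speed-weighted additive sprint and base poses when not skydiving.
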AnimBlueprint /BRCosmetics/Accessories/FORT_Backpacks/F_MED_DarkViking_Backpack/Meshes/F_MED_Commando_DarkViking_Backpack_AnimBP
target_skeleton=/Game/Accessories/FORT_Backpacks/F_MED_DarkViking_Backpack/Meshes/F_MED_Commando_DarkViking_Backpack_Skeleton
graphs [
 AnimGraph {nodes [
   n1 = MapRangeClamped(Value=Get PawnSpeed2D(), InRangeA=0, InRangeB=549)
   n2 = F_MED_Commando_DarkViking_Backpack_Dive_PoseAsset : sprint!(PoseWeight:float=@n1) -> (Pose:Pose Link Structure) {
    is=title AnimGraphNode_PoseByName
    poseAsset=&/F_MED_Commando_DarkViking_Backpack_Dive_PoseAsset
    poseName=sprint
   }
   n3 = Apply Additive(Base=Use cached pose 'preSprint'(), Additive=@n2, Alpha=0.33) {alphaBoolBlend {blendOption=Linear}}
   n4 = F_MED_Commando_DarkViking_Backpack_Base_PoseAsset : base!(PoseWeight:float=@n1) -> (Pose:Pose Link Structure) {
    is=title AnimGraphNode_PoseByName
    poseAsset=&/F_MED_Commando_DarkViking_Backpack_Base_PoseAsset
    poseName=base
   }
   Output Pose(Result=Blend Poses by bool(
     BlendPose_0=Use cached pose 'preSprint'()
     BlendPose_1=Apply Additive(Base=@n3, Additive=@n4) {alphaBoolBlend {blendOption=Linear}}
     BlendTime_0=0.3
     BlendTime_1=0.3
     bActiveValue=Get bIsSkyDiving()
    ) {blendType=Linear}) {layerGroup=None}
   n5 = "float >= float"(A=Get PawnSpeed2D(), B=400) {is=title K2Node_CallFunction}
   n6 = SelectFloat(A=0.05, B=SelectFloat(A=0.5, B=0.2, bPickA=@n5), bPickA=Get bIsSkydiveDiveMode())
   n7 = Get falling_anim_dyn_up_force_vec3() -> (falling_anim_dyn_up_force_vec3:Vector)
   n8 = SelectFloat(A=0.1, B=SelectFloat(A=0.6, B=0.3, bPickA=@n5), bPickA=Get bIsSkydiveDiveMode())
   AnimGraphNode_StateMachine_0 = wingPosition() {is=title AnimGraphNode_StateMachine}
   n9 = Anim Dynamics - Bone: dyn_wing_wrist_l(ComponentPose=Local To Component(LocalPose=@AnimGraphNode_StateMachine_0), Alpha=@n8, ExternalForce="0,0,0") {
    alpha=0.9
    angularDampingOverride=0.98
    angularSpringConstant=500
    bGravityOverrideInSimSpace=true
    bOverrideAngularDamping=true
    bOverrideLinearDamping=true
    gravityScale=0.125
    linearDampingOverride=0.98
    numSolverIterationsPostUpdate=4
    numSolverIterationsPreUpdate=16
    alphaBoolBlend {blendOption=Linear}
    boundBone {boneName=dyn_wing_wrist_l}
    componentLinearAccScale=<0.125 0.125 0.125>
    componentLinearVelScale=<0.125 0.125 0.125>
    physicsBodyDefinitions [{
      sphereCollisionRadius=0
      boundBone {boneName=dyn_wing_wrist_l}
      boxExtents=<10 10 5>
      constraintSetup {angularTargetAxis=AxisZ, twistAxis=AxisZ, angularLimitsMax=<0 15 0>, angularLimitsMin=<0 -30 0>, angularTarget=<0 0 1>}
      localJointOffset=<0 0 1>
     }]
   }
   n10 = Anim Dynamics - Bone: dyn_wing_wrist_r(ComponentPose=@n9, Alpha=@n8, ExternalForce="0,0,0") {
    alpha=0.9
    angularDampingOverride=0.98
    angularSpringConstant=500
    bGravityOverrideInSimSpace=true
    bOverrideAngularDamping=true
    bOverrideLinearDamping=true
    gravityScale=0.125
    linearDampingOverride=0.98
    numSolverIterationsPostUpdate=4
    numSolverIterationsPreUpdate=16
    alphaBoolBlend {blendOption=Linear}
    boundBone {boneName=dyn_wing_wrist_r}
    componentLinearAccScale=<0.125 0.125 0.125>
    componentLinearVelScale=<0.125 0.125 0.125>
    physicsBodyDefinitions [{
      sphereCollisionRadius=0
      boundBone {boneName=dyn_wing_wrist_r}
      boxExtents=<10 10 5>
      constraintSetup {angularTargetAxis=AxisZ, twistAxis=AxisZ, angularLimitsMax=<0 15 0>, angularLimitsMin=<0 -30 0>, angularTarget=<0 0 1>}
      localJointOffset=<0 0 -1>
     }]
   }
   n11 = Anim Dynamics - Bone: dyn_wing_pointer_l_01(ComponentPose=@n10, Alpha=@n8, ExternalForce="0,0,0") {
    alpha=0.9
    angularDampingOverride=0.98
    angularSpringConstant=500
    bAngularSpring=true
    bGravityOverrideInSimSpace=true
    bOverrideAngularDamping=true
    gravityScale=0.125
    linearDampingOverride=0.98
    numSolverIterationsPostUpdate=4
    numSolverIterationsPreUpdate=16
    alphaBoolBlend {blendOption=Linear}
    boundBone {boneName=dyn_wing_pointer_l_01}
    componentLinearAccScale=<0.04 0.04 0.04>
    componentLinearVelScale=<0.3 0.3 0.3>
    physicsBodyDefinitions [{
      sphereCollisionRadius=0
      boundBone {boneName=dyn_wing_pointer_l_01}
      boxExtents=<10 10 5>
      constraintSetup {angularTargetAxis=AxisZ, twistAxis=AxisZ, angularLimitsMax=<0 5 0>, angularLimitsMin=<0 -80 0>, angularTarget=<0 0 1>}
      localJointOffset=<0 0 1>
     }]
   }
   n12 = Anim Dynamics - Bone: dyn_wing_middle_l_01(ComponentPose=@n11, Alpha=@n8, ExternalForce="0,0,0") {
    alpha=0.9
    angularDampingOverride=0.98
    angularSpringConstant=500
    bAngularSpring=true
    bGravityOverrideInSimSpace=true
    bOverrideAngularDamping=true
    gravityScale=0.125
    linearDampingOverride=0.98
    numSolverIterationsPostUpdate=4
    numSolverIterationsPreUpdate=16
    alphaBoolBlend {blendOption=Linear}
    boundBone {boneName=dyn_wing_middle_l_01}
    componentLinearAccScale=<0.06 0.06 0.06>
    componentLinearVelScale=<0.4 0.4 0.4>
    physicsBodyDefinitions [{
      sphereCollisionRadius=0
      boundBone {boneName=dyn_wing_middle_l_01}
      boxExtents=<10 10 5>
      constraintSetup {angularTargetAxis=AxisZ, twistAxis=AxisZ, angularLimitsMax=<0 10 0>, angularLimitsMin=<0 -80 0>, angularTarget=<0 0 1>}
      localJointOffset=<0 0 1>
     }]
   }
   n13 = Anim Dynamics - Bone: dyn_wing_pinky_l_01(ComponentPose=@n12, Alpha=@n8, ExternalForce="0,0,0") {
    alpha=0.9
    angularDampingOverride=0.98
    angularSpringConstant=500
    bAngularSpring=true
    bGravityOverrideInSimSpace=true
    bOverrideAngularDamping=true
    gravityScale=0.125
    linearDampingOverride=0.98
    numSolverIterationsPostUpdate=4
    numSolverIterationsPreUpdate=16
    alphaBoolBlend {blendOption=Linear}
    boundBone {boneName=dyn_wing_pinky_l_01}
    componentLinearAccScale=<0.08 0.08 0.08>
    componentLinearVelScale=<0.5 0.5 0.5>
    physicsBodyDefinitions [{
      sphereCollisionRadius=0
      boundBone {boneName=dyn_wing_pinky_l_01}
      boxExtents=<10 10 5>
      constraintSetup {angularTargetAxis=AxisZ, twistAxis=AxisZ, angularLimitsMax=<0 40 0>, angularLimitsMin=<0 -80 0>, angularTarget=<0 0 1>}
      localJointOffset=<0 0 1>
     }]
   }
   n14 = Anim Dynamics - Bone: dyn_wing_pointer_r_01(ComponentPose=@n13, Alpha=@n8, ExternalForce="0,0,0") {
    alpha=0.9
    angularDampingOverride=0.98
    angularSpringConstant=500
    bAngularSpring=true
    bGravityOverrideInSimSpace=true
    bOverrideAngularDamping=true
    gravityScale=0.125
    linearDampingOverride=0.98
    numSolverIterationsPostUpdate=4
    numSolverIterationsPreUpdate=16
    alphaBoolBlend {blendOption=Linear}
    boundBone {boneName=dyn_wing_pointer_r_01}
    componentLinearAccScale=<0.04 0.04 0.04>
    componentLinearVelScale=<0.3 0.3 0.3>
    physicsBodyDefinitions [{
      sphereCollisionRadius=0
      boundBone {boneName=dyn_wing_pointer_r_01}
      boxExtents=<10 10 5>
      constraintSetup {angularTargetAxis=AxisZ, twistAxis=AxisZ, angularLimitsMax=<0 5 0>, angularLimitsMin=<0 -80 0>, angularTarget=<0 0 1>}
      localJointOffset=<0 0 -1>
     }]
   }
   n15 = Anim Dynamics - Bone: dyn_wing_middle_r_01(ComponentPose=@n14, Alpha=@n8, ExternalForce="0,0,0") {
    alpha=0.9
    angularDampingOverride=0.98
    angularSpringConstant=500
    bAngularSpring=true
    bGravityOverrideInSimSpace=true
    bOverrideAngularDamping=true
    gravityScale=0.125
    linearDampingOverride=0.98
    numSolverIterationsPostUpdate=4
    numSolverIterationsPreUpdate=16
    alphaBoolBlend {blendOption=Linear}
    boundBone {boneName=dyn_wing_middle_r_01}
    componentLinearAccScale=<0.06 0.06 0.06>
    componentLinearVelScale=<0.4 0.4 0.4>
    physicsBodyDefinitions [{
      sphereCollisionRadius=0
      boundBone {boneName=dyn_wing_middle_r_01}
      boxExtents=<10 10 5>
      constraintSetup {angularTargetAxis=AxisZ, twistAxis=AxisZ, angularLimitsMax=<0 10 0>, angularLimitsMin=<0 -80 0>, angularTarget=<0 0 1>}
      localJointOffset=<0 0 -1>
     }]
   }
   n16 = Anim Dynamics - Bone: dyn_wing_pinky_r_01(ComponentPose=@n15, Alpha=@n8, ExternalForce="0,0,0") {
    alpha=0.9
    angularDampingOverride=0.98
    angularSpringConstant=500
    bAngularSpring=true
    bGravityOverrideInSimSpace=true
    bOverrideAngularDamping=true
    gravityScale=0.125
    linearDampingOverride=0.98
    numSolverIterationsPostUpdate=4
    numSolverIterationsPreUpdate=16
    alphaBoolBlend {blendOption=Linear}
    boundBone {boneName=dyn_wing_pinky_r_01}
    componentLinearAccScale=<0.08 0.08 0.08>
    componentLinearVelScale=<0.5 0.5 0.5>
    physicsBodyDefinitions [{
      sphereCollisionRadius=0
      boundBone {boneName=dyn_wing_pinky_r_01}
      boxExtents=<10 10 5>
      constraintSetup {angularTargetAxis=AxisZ, twistAxis=AxisZ, angularLimitsMax=<0 40 0>, angularLimitsMin=<0 -80 0>, angularTarget=<0 0 1>}
      localJointOffset=<0 0 -1>
     }]
   }
   n17 = Anim Dynamics - Bone: dyn_wing_pointer_l_02(ComponentPose=@n16, Alpha=@n6, ExternalForce=@n7) {
    alpha=0.9
    angularDampingOverride=0.98
    angularSpringConstant=500
    bAngularSpring=true
    bGravityOverrideInSimSpace=true
    bOverrideAngularDamping=true
    gravityScale=0.125
    linearDampingOverride=0.98
    numSolverIterationsPostUpdate=4
    numSolverIterationsPreUpdate=16
    alphaBoolBlend {blendOption=Linear}
    boundBone {boneName=dyn_wing_pointer_l_02}
    componentLinearAccScale=<0.04 0.04 0.04>
    componentLinearVelScale=<0.3 0.3 0.3>
    physicsBodyDefinitions [{
      sphereCollisionRadius=0
      boundBone {boneName=dyn_wing_pointer_l_02}
      boxExtents=<10 10 5>
      constraintSetup {angularTargetAxis=AxisZ, twistAxis=AxisZ, angularLimitsMax=<15 5 0>, angularLimitsMin=<-15 -80 0>, angularTarget=<0 0 1>}
      localJointOffset=<0 0 1>
     }]
   }
   n18 = Anim Dynamics - Bone: dyn_wing_middle_l_02(ComponentPose=@n17, Alpha=@n6, ExternalForce=@n7) {
    alpha=0.9
    angularDampingOverride=0.98
    angularSpringConstant=500
    bAngularSpring=true
    bGravityOverrideInSimSpace=true
    bOverrideAngularDamping=true
    gravityScale=0.125
    linearDampingOverride=0.98
    numSolverIterationsPostUpdate=4
    numSolverIterationsPreUpdate=16
    alphaBoolBlend {blendOption=Linear}
    boundBone {boneName=dyn_wing_middle_l_02}
    componentLinearAccScale=<0.06 0.06 0.06>
    componentLinearVelScale=<0.4 0.4 0.4>
    physicsBodyDefinitions [{
      sphereCollisionRadius=0
      boundBone {boneName=dyn_wing_middle_l_02}
      boxExtents=<10 10 5>
      constraintSetup {angularTargetAxis=AxisZ, twistAxis=AxisZ, angularLimitsMax=<15 10 0>, angularLimitsMin=<-15 -80 0>, angularTarget=<0 0 1>}
      localJointOffset=<0 0 1>
     }]
   }
   n19 = Anim Dynamics - Bone: dyn_wing_pointer_r_02(ComponentPose=@n18, Alpha=@n6, ExternalForce=@n7) {
    alpha=0.9
    angularDampingOverride=0.98
    angularSpringConstant=500
    bAngularSpring=true
    bGravityOverrideInSimSpace=true
    bOverrideAngularDamping=true
    gravityScale=0.125
    linearDampingOverride=0.98
    numSolverIterationsPostUpdate=4
    numSolverIterationsPreUpdate=16
    alphaBoolBlend {blendOption=Linear}
    boundBone {boneName=dyn_wing_pointer_r_02}
    componentLinearAccScale=<0.04 0.04 0.04>
    componentLinearVelScale=<0.3 0.3 0.3>
    physicsBodyDefinitions [{
      sphereCollisionRadius=0
      boundBone {boneName=dyn_wing_pointer_r_02}
      boxExtents=<10 10 5>
      constraintSetup {angularTargetAxis=AxisZ, twistAxis=AxisZ, angularLimitsMax=<15 5 0>, angularLimitsMin=<-15 -80 0>, angularTarget=<0 0 1>}
      localJointOffset=<0 0 -1>
     }]
   }
   n20 = Anim Dynamics - Bone: dyn_wing_middle_r_02(ComponentPose=@n19, Alpha=@n6, ExternalForce=@n7) {
    alpha=0.9
    angularDampingOverride=0.98
    angularSpringConstant=500
    bAngularSpring=true
    bGravityOverrideInSimSpace=true
    bOverrideAngularDamping=true
    gravityScale=0.125
    linearDampingOverride=0.98
    numSolverIterationsPostUpdate=4
    numSolverIterationsPreUpdate=16
    alphaBoolBlend {blendOption=Linear}
    boundBone {boneName=dyn_wing_middle_r_02}
    componentLinearAccScale=<0.06 0.06 0.06>
    componentLinearVelScale=<0.4 0.4 0.4>
    physicsBodyDefinitions [{
      sphereCollisionRadius=0
      boundBone {boneName=dyn_wing_middle_r_02}
      boxExtents=<10 10 5>
      constraintSetup {angularTargetAxis=AxisZ, twistAxis=AxisZ, angularLimitsMax=<15 10 0>, angularLimitsMin=<-15 -80 0>, angularTarget=<0 0 1>}
      localJointOffset=<0 0 -1>
     }]
   }
   n21 = Anim Dynamics - Bone: dyn_wing_pointer_l_03(ComponentPose=@n20, Alpha=@n6, ExternalForce=@n7) {
    alpha=0.9
    angularDampingOverride=0.98
    angularSpringConstant=500
    bAngularSpring=true
    bGravityOverrideInSimSpace=true
    bOverrideAngularDamping=true
    gravityScale=0.125
    linearDampingOverride=0.98
    numSolverIterationsPostUpdate=4
    numSolverIterationsPreUpdate=16
    alphaBoolBlend {blendOption=Linear}
    boundBone {boneName=dyn_wing_pointer_l_03}
    componentLinearAccScale=<0.04 0.04 0.04>
    componentLinearVelScale=<0.3 0.3 0.3>
    physicsBodyDefinitions [{
      sphereCollisionRadius=0
      boundBone {boneName=dyn_wing_pointer_l_03}
      boxExtents=<10 10 5>
      constraintSetup {angularTargetAxis=AxisZ, twistAxis=AxisZ, angularLimitsMax=<15 5 15>, angularLimitsMin=<-15 -80 -15>, angularTarget=<0 0 1>}
      localJointOffset=<0 0 1>
     }]
   }
   n22 = Anim Dynamics - Bone: dyn_wing_middle_l_03(ComponentPose=@n21, Alpha=@n6, ExternalForce=@n7) {
    alpha=0.9
    angularDampingOverride=0.98
    angularSpringConstant=500
    bAngularSpring=true
    bGravityOverrideInSimSpace=true
    bOverrideAngularDamping=true
    gravityScale=0.125
    linearDampingOverride=0.98
    numSolverIterationsPostUpdate=4
    numSolverIterationsPreUpdate=16
    alphaBoolBlend {blendOption=Linear}
    boundBone {boneName=dyn_wing_middle_l_03}
    componentLinearAccScale=<0.06 0.06 0.06>
    componentLinearVelScale=<0.4 0.4 0.4>
    physicsBodyDefinitions [{
      sphereCollisionRadius=0
      boundBone {boneName=dyn_wing_middle_l_03}
      boxExtents=<10 10 5>
      constraintSetup {angularTargetAxis=AxisZ, twistAxis=AxisZ, angularLimitsMax=<15 10 15>, angularLimitsMin=<-15 -80 -15>, angularTarget=<0 0 1>}
      localJointOffset=<0 0 1>
     }]
   }
   n23 = Anim Dynamics - Bone: dyn_wing_pointer_r_03(ComponentPose=@n22, Alpha=@n6, ExternalForce=@n7) {
    alpha=0.9
    angularDampingOverride=0.98
    angularSpringConstant=500
    bAngularSpring=true
    bGravityOverrideInSimSpace=true
    bOverrideAngularDamping=true
    gravityScale=0.125
    linearDampingOverride=0.98
    numSolverIterationsPostUpdate=4
    numSolverIterationsPreUpdate=16
    alphaBoolBlend {blendOption=Linear}
    boundBone {boneName=dyn_wing_pointer_r_03}
    componentLinearAccScale=<0.04 0.04 0.04>
    componentLinearVelScale=<0.3 0.3 0.3>
    physicsBodyDefinitions [{
      sphereCollisionRadius=0
      boundBone {boneName=dyn_wing_pointer_r_03}
      boxExtents=<10 10 5>
      constraintSetup {angularTargetAxis=AxisZ, twistAxis=AxisZ, angularLimitsMax=<15 5 15>, angularLimitsMin=<-15 -80 -15>, angularTarget=<0 0 1>}
      localJointOffset=<0 0 -1>
     }]
   }
   n24 = Anim Dynamics - Bone: dyn_wing_middle_r_03(ComponentPose=@n23, Alpha=@n6, ExternalForce=@n7) {
    alpha=0.9
    angularDampingOverride=0.98
    angularSpringConstant=500
    bAngularSpring=true
    bGravityOverrideInSimSpace=true
    bOverrideAngularDamping=true
    gravityScale=0.125
    linearDampingOverride=0.98
    numSolverIterationsPostUpdate=4
    numSolverIterationsPreUpdate=16
    alphaBoolBlend {blendOption=Linear}
    boundBone {boneName=dyn_wing_middle_r_03}
    componentLinearAccScale=<0.06 0.06 0.06>
    componentLinearVelScale=<0.4 0.4 0.4>
    physicsBodyDefinitions [{
      sphereCollisionRadius=0
      boundBone {boneName=dyn_wing_middle_r_03}
      boxExtents=<10 10 5>
      constraintSetup {angularTargetAxis=AxisZ, twistAxis=AxisZ, angularLimitsMax=<15 10 15>, angularLimitsMin=<-15 -80 -15>, angularTarget=<0 0 1>}
      localJointOffset=<0 0 -1>
     }]
   }
   Save cached pose 'preSprint'(Pose=Component To Local(ComponentPose=@n24)) {cacheName=preSprint}
  ]}
 wingPosition {
  entry=Base
  path=AnimGraph.AnimGraphNode_StateMachine_0.wingPosition
  ids=id
  nodes [
   AnimStateNode_0 = Base() {is=title AnimStateNode}
   AnimStateNode_1 = Diving() {is=title AnimStateNode}
   AnimStateNode_2 = Floating() {is=title AnimStateNode}
   AnimStateNode_3 = Targeting() {is=title AnimStateNode}
   AnimStateTransitionNode_0 = AnimStateTransitionNode()
   AnimStateTransitionNode_1 = AnimStateTransitionNode()
   AnimStateTransitionNode_2 = AnimStateTransitionNode()
   AnimStateTransitionNode_5 = AnimStateTransitionNode()
   AnimStateTransitionNode_6 = AnimStateTransitionNode()
   AnimStateTransitionNode_7 = AnimStateTransitionNode()
  ]
  links [
   {
    dst=Diving
    rule=AnimGraph.AnimGraphNode_StateMachine_0.wingPosition.AnimStateTransitionNode_0.Transition
    src=Base
    via=AnimStateTransitionNode_0
   }
   {
    dst=Targeting
    rule=AnimGraph.AnimGraphNode_StateMachine_0.wingPosition.AnimStateTransitionNode_6.Transition
    src=Base
    via=AnimStateTransitionNode_6
   }
   {
    dst=Floating
    rule=AnimGraph.AnimGraphNode_StateMachine_0.wingPosition.AnimStateTransitionNode_1.Transition
    src=Diving
    via=AnimStateTransitionNode_1
   }
   {
    dst=Base
    rule=AnimGraph.AnimGraphNode_StateMachine_0.wingPosition.AnimStateTransitionNode_2.Transition
    src=Floating
    via=AnimStateTransitionNode_2
   }
   {
    dst=Diving
    rule=AnimGraph.AnimGraphNode_StateMachine_0.wingPosition.AnimStateTransitionNode_5.Transition
    src=Floating
    via=AnimStateTransitionNode_5
   }
   {
    dst=Base
    rule=AnimGraph.AnimGraphNode_StateMachine_0.wingPosition.AnimStateTransitionNode_7.Transition
    src=Targeting
    via=AnimStateTransitionNode_7
   }
  ]
 }
 Base {
  path=AnimGraph.AnimGraphNode_StateMachine_0.wingPosition.AnimStateNode_0.Base
  nodes [Output Animation Pose(Result=Sequence Player 'F_MED_Commando_DarkViking_Backpack_Base_01'() {sequence=&/F_MED_Commando_DarkViking_Backpack_Base_01})]
 }
 Diving {
  path=AnimGraph.AnimGraphNode_StateMachine_0.wingPosition.AnimStateNode_1.Diving
  nodes [Output Animation Pose(Result=Sequence Player 'F_MED_Commando_DarkViking_Backpack_Dive_01'() {sequence=&/F_MED_Commando_DarkViking_Backpack_Dive_01})]
 }
 Floating {
  path=AnimGraph.AnimGraphNode_StateMachine_0.wingPosition.AnimStateNode_2.Floating
  nodes [
   Sequence Player 'F_MED_Commando_DarkViking_Backpack_Fly_01'() {sequence=/Game/Accessories/FORT_Backpacks/F_MED_DarkViking_Backpack/Meshes/F_MED_Commando_DarkViking_Backpack_Fly_01}
   Output Animation Pose(Result=Sequence Player 'F_MED_Commando_DarkViking_Backpack_Fly_02'() {sequence=&/F_MED_Commando_DarkViking_Backpack_Fly_02})
  ]
 }
 Targeting {
  path=AnimGraph.AnimGraphNode_StateMachine_0.wingPosition.AnimStateNode_3.Targeting
  nodes [Output Animation Pose(Result=Sequence Player 'F_MED_Commando_DarkViking_Backpack_Dive_01'() {sequence=&F_MED_Commando_DarkViking_Backpack_Dive_01})]
 }
 Transition {
  path=AnimGraph.AnimGraphNode_StateMachine_0.wingPosition.AnimStateTransitionNode_0.Transition
  nodes [Result(bCanEnterTransition=Get bIsSkydiveDiveMode())]
 }
 Transition {
  path=AnimGraph.AnimGraphNode_StateMachine_0.wingPosition.AnimStateTransitionNode_1.Transition
  nodes [Result(bCanEnterTransition=NOT Boolean(A=Get bIsSkydiveDiveMode()))]
 }
 Transition {
  path=AnimGraph.AnimGraphNode_StateMachine_0.wingPosition.AnimStateTransitionNode_2.Transition
  nodes [Result(bCanEnterTransition=OR Boolean(A=Get bIsOnGround() -> (bIsOnGround:bool), B=Get bIsParachuteOpened()))]
 }
 Transition {
  path=AnimGraph.AnimGraphNode_StateMachine_0.wingPosition.AnimStateTransitionNode_5.Transition
  nodes [Result(bCanEnterTransition=Get bIsSkydiveDiveMode())]
 }
 Transition {
  path=AnimGraph.AnimGraphNode_StateMachine_0.wingPosition.AnimStateTransitionNode_6.Transition
  nodes [Result(bCanEnterTransition=Get bIsTargeting())]
 }
 Transition {
  path=AnimGraph.AnimGraphNode_StateMachine_0.wingPosition.AnimStateTransitionNode_7.Transition
  nodes [Result(bCanEnterTransition=NOT Boolean(A=Get bIsTargeting()))]
 }
 EventGraph {nodes [
   n1 = Event BlueprintUpdateAnimation()
   > Set deltaTimeX(deltaTimeX:double=@n1.DeltaTimeX) -> (Output_Get:double)
   n2 = Get PawnLastLocation() -> (PawnLastLocation:Vector)
   n3 = SelectFloat(
    A=Make Perlin Noise and Remap!(Value:float=BreakVector(InVec=@n2).Z, RangeOutMin:float=-3500, RangeOutMax:float=3500) -> (ReturnValue:float)
    B=0
    bPickA=Get bIsSkyDiving()
   )
   n4 = SelectFloat(A=float * float(A=@n3, B=0.5), B=@n3, bPickA=Get bIsParachuteOpened())
   n5 = MakeVector(X=@n4, Y=@n4)
   > Set falling_anim_dyn_up_force_vec3(falling_anim_dyn_up_force_vec3:Vector=@n5) -> (Output_Get:Vector)
   TryGetPawnOwner()
   SelectVector(A="0, 0, 0", B=@n5, bPickA=Get bIsSkydiveDiveMode())
  ]}
]
variables [
 deltaTimeX {type=number}
 falling_anim_dyn_up_force_vec3 {type=Vector}
]
